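# Sound cue for the LoveLlama cosmetic pickaxe swing.
# A quality-level root selects one of three mixers with progressively fewer wave variations, the last two levels sharing the smallest.
# Each mixer layers random LoveLlama swings under close attenuation with Fire Axe distant swings under medium attenuation.
# The distant layer passes through a modulator, and the smallest tier uses one distant wave with a wider pitch range for variety.
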SoundCue /BRCosmetics/Sounds/PickAxes/LoveLlama/Pickaxe_LoveLlama_Swing_Cue
root=QualityLevel_1
nodes [
 {id=Attenuation_10, children [Random_12], properties {attenuationSettings=/Game/Sounds/Attenuation/Fort_Weapon_Pickaxe_Close_Att}}
 {id=Attenuation_11, children [Modulator_7], properties {attenuationSettings=/Game/Sounds/Attenuation/Fort_Weapon_Pickaxe_Medium_Att}}
 {id=Attenuation_6, children [Random_0], properties {attenuationSettings=&Fort_Weapon_Pickaxe_Close_Att}}
 {id=Attenuation_7, children [Random_2], properties {attenuationSettings=&Fort_Weapon_Pickaxe_Medium_Att}}
 {id=Attenuation_8, children [Random_3], properties {attenuationSettings=&Fort_Weapon_Pickaxe_Close_Att}}
 {id=Attenuation_9, children [Random_4], properties {attenuationSettings=&Fort_Weapon_Pickaxe_Medium_Att}}
 {id=Mixer_3, children [Attenuation_6, Modulator_0], properties {inputVolume [1,1]}}
 {id=Mixer_4, children [Attenuation_8, Modulator_5], properties {inputVolume [1,1]}}
 {id=Mixer_5, children [Attenuation_10, Attenuation_11], properties {inputVolume [1,1]}}
 {id=Modulator_0, children [Attenuation_7], properties {pitchMax=1.03, pitchMin=0.97}}
 {id=Modulator_5, children [Attenuation_9]}
 {id=Modulator_7, children [WavePlayer_31], properties {pitchMax=1.2, pitchMin=0.9}}
 {id=QualityLevel_1, children [Mixer_3,Mixer_4,Mixer_5,Mixer_5]}
 {
  id=Random_0
  children [WavePlayer_0,WavePlayer_1,WavePlayer_2,WavePlayer_3,WavePlayer_4,WavePlayer_5,WavePlayer_6,WavePlayer_7]
  properties {weights [1,1,1,1,1,1,1,1]}
 }
 {id=Random_12, children [WavePlayer_49, WavePlayer_50], properties {weights [1,1]}}
 {
  id=Random_2
  children [WavePlayer_10, WavePlayer_11, WavePlayer_14, WavePlayer_15, WavePlayer_17, WavePlayer_21]
  properties {weights [1,1,1,1,1,1]}
 }
 {id=Random_3, children [WavePlayer_46, WavePlayer_47, WavePlayer_48], properties {weights [1,1,1]}}
 {id=Random_4, children [WavePlayer_23, WavePlayer_22, WavePlayer_24], properties {weights [1,1,1]}}
 {id=WavePlayer_0, properties {soundWaveAssetPtr=&/Pickaxe_LoveLlama_Swing_01}}
 {id=WavePlayer_1, properties {soundWaveAssetPtr=&/Pickaxe_LoveLlama_Swing_02}}
 {id=WavePlayer_10, properties {soundWaveAssetPtr=/Game/Athena/Sounds/Weapons/Fire_Axe_Swing_Distant_04}}
 {id=WavePlayer_11, properties {soundWaveAssetPtr=/Game/Athena/Sounds/Weapons/Fire_Axe_Swing_Distant_01}}
 {id=WavePlayer_14, properties {soundWaveAssetPtr=/Game/Athena/Sounds/Weapons/Fire_Axe_Swing_Distant_02}}
 {id=WavePlayer_15, properties {soundWaveAssetPtr=/Game/Athena/Sounds/Weapons/Fire_Axe_Swing_Distant_03}}
 {id=WavePlayer_17, properties {soundWaveAssetPtr=/Game/Athena/Sounds/Weapons/Fire_Axe_Swing_Distant_05}}
 {id=WavePlayer_2, properties {soundWaveAssetPtr=&/Pickaxe_LoveLlama_Swing_03}}
 {id=WavePlayer_21, properties {soundWaveAssetPtr=/Game/Athena/Sounds/Weapons/Fire_Axe_Swing_Distant_06}}
 {id=WavePlayer_22, properties {soundWaveAssetPtr=&Fire_Axe_Swing_Distant_03}}
 {id=WavePlayer_23, properties {soundWaveAssetPtr=&Fire_Axe_Swing_Distant_04}}
 {id=WavePlayer_24, properties {soundWaveAssetPtr=&Fire_Axe_Swing_Distant_06}}
 {id=WavePlayer_3, properties {soundWaveAssetPtr=&/Pickaxe_LoveLlama_Swing_04}}
 {id=WavePlayer_31, properties {soundWaveAssetPtr=&Fire_Axe_Swing_Distant_06}}
 {id=WavePlayer_4, properties {soundWaveAssetPtr=&/Pickaxe_LoveLlama_Swing_05}}
 {id=WavePlayer_46, properties {soundWaveAssetPtr=&Pickaxe_LoveLlama_Swing_01}}
 {id=WavePlayer_47, properties {soundWaveAssetPtr=&Pickaxe_LoveLlama_Swing_02}}
 {id=WavePlayer_48, properties {soundWaveAssetPtr=&Pickaxe_LoveLlama_Swing_03}}
 {id=WavePlayer_49, properties {soundWaveAssetPtr=&Pickaxe_LoveLlama_Swing_01}}
 {id=WavePlayer_5, properties {soundWaveAssetPtr=&/Pickaxe_LoveLlama_Swing_06}}
 {id=WavePlayer_50, properties {soundWaveAssetPtr=&Pickaxe_LoveLlama_Swing_02}}
 {id=WavePlayer_6, properties {soundWaveAssetPtr=&/Pickaxe_LoveLlama_Swing_07}}
 {id=WavePlayer_7, properties {soundWaveAssetPtr=&/Pickaxe_LoveLlama_Swing_08}}
]
properties {
 audioPropertiesSheet=/BRCosmetics/Sounds/Settings/PropertySheets/Pickaxe/APS_Cosmetic_Pickaxe_Swing
 bPrimeOnLoad=true
 maxDistance=3000
 soundClassObject=&Fort_WeaponMelee_PSM
 volumeMultiplier=0.3
}
